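# Sound cue for the electrical zap heard when a faction chest scan fails.
# A single wave player for the scan-fail zap sample feeds a modulator that adds slight random pitch variation.
# It plays on the loot reward sound class with chest attenuation.
# A shared concurrency rule limits simultaneous copies, stopping the furthest or oldest.
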SoundCue /FNGameplayCues/Sounds/Foley_Loot/Containers/Faction_Chest/Faction_Chest_Fail_ElecZap_Cue
root=Modulator_0
nodes [
 {id=Modulator_0, children [WavePlayer_8], properties {pitchMax=1.03, pitchMin=0.99}}
 {id=WavePlayer_8, properties {soundWaveAssetPtr=&/Faction_Chest_Scan_Fail_Zap}}
]
properties {
 attenuationSettings=/Game/Sounds/Attenuation/Fort_Chests_Att
 maxDistance=1300
 soundClassObject=&Fort_Loot_Reward_PSM
 volumeMultiplier=0.6
 attenuationOverrides {
  bEnableOcclusion=true
  falloffDistance=1000
  occlusionInterpolationTime=0.3
  occlusionLowPassFilterFrequency=3000
  occlusionTraceChannel=ECC_GameTraceChannel11
  occlusionVolumeAttenuation=0.75
  attenuationShapeExtents {x=500}
 }
 concurrencyOverrides {maxCount=4}
 concurrencySet [/Game/Sounds/Concurrency_Rules/Demo_FourMaximum_StopFurthestOldest_VolumeScale]
}
