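# Sound cue for a failed scan on a faction chest, in the container loot foley set.
# A Mixer plays two waves at once, the scan fail sound at reduced volume and the eye scan fail sound at full volume.
# The cue is spatialized through the shared Fort_Chests_Att attenuation asset and routed to the loot reward sound class.
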SoundCue /FNGameplayCues/Sounds/Foley_Loot/Containers/Faction_Chest/Faction_Chest_Failed_Cue
root=Mixer_0
nodes [
 {id=Mixer_0, children [WavePlayer_0,WavePlayer_1], properties {inputVolume [0.7,1]}}
 {id=WavePlayer_0, properties {soundWaveAssetPtr=&/Faction_Chest_Scan_Fail}}
 {id=WavePlayer_1, properties {soundWaveAssetPtr=&/Faction_Chest_Eye_Scan_Fail_01}}
]
properties {
 attenuationSettings=/Game/Sounds/Attenuation/Fort_Chests_Att
 maxDistance=1300
 soundClassObject=&Fort_Loot_Reward_PSM
 volumeMultiplier=1
 attenuationOverrides {
  bEnableOcclusion=true
  falloffDistance=1000
  occlusionInterpolationTime=0.3
  occlusionLowPassFilterFrequency=3000
  occlusionTraceChannel=ECC_GameTraceChannel11
  occlusionVolumeAttenuation=0.75
  attenuationShapeExtents {x=500}
 }
}
